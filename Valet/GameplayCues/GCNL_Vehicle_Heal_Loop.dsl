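; Looping gameplay cue notify for a vehicle heal effect, currently only a debug stub.
; On application it prints "Apply" in red, and on removal it prints "removed" in green.
; The event graphs ignore the target, cue parameters and spawn results, and the construction script is empty.
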(event EventOnApplication
  (params (Target :type Actor%20Object%20Reference) (Parameters :type Gameplay%20Cue%20Parameters%20Structure%20%28by%20ref%29) (SpawnResults :type Gameplay%20Cue%20Notify%20Spawn%20Result%20Structure%20%28by%20ref%29))
  (Development|PrintString 0 "Apply" true true "(R=1.000000,G=0.009999,B=0.000000,A=1.000000)"))

(event EventOnRemoval
  (params (Target :type Actor%20Object%20Reference) (Parameters :type Gameplay%20Cue%20Parameters%20Structure%20%28by%20ref%29) (SpawnResults :type Gameplay%20Cue%20Notify%20Spawn%20Result%20Structure%20%28by%20ref%29))
  (Development|PrintString 0 "removed" true true "(R=0.153796,G=1.000000,B=0.000000,A=1.000000)"))

(fn ConstructionScript)
